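# Physics asset for the Sculptor Ice gadget prop's skeletal mesh, giving it a simple collision shape.
# A single body on bone JNT_Year_23 holds one capsule, offset and rotated from the bone origin.
# No constraints are defined, so the asset sets up no jointed or multi-body ragdoll motion.
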PhysicsAsset /Game/Animation/Props/Gadgets_Sculptor_Ice/Meshes/Sculptor_Ice_Gadget_Physics
bodies {JNT_Year_23 [{shapeType=Capsule, length=39.3668, radius=30.9077, center=<-3.4697 0.6813 36.2509>, rotation=<rot 0.8659 -90 -5.75>}]}
constraints []
